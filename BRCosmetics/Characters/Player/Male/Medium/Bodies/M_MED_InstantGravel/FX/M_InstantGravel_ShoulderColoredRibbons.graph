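# Translucent unlit Niagara ribbon material for the InstantGravel character's shoulder trails, colored by particle color through the shared emissive adjustment function.
# A trail texture panned on UV1, offset in U by a dynamic parameter, is shaped by particle alpha and UV0 base, tip and side softness masks.
# Depth fade applies only on default and deferred shading paths.
# TrailUV_Settings uses ScaleX and PanX for both axes, leaving ScaleY and PanY unused.
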Material /BRCosmetics/Characters/Player/Male/Medium/Bodies/M_MED_InstantGravel/FX/M_InstantGravel_ShoulderColoredRibbons
graph {
 properties {bUsedWithNiagaraRibbons=true, blendMode=BLEND_Translucent, pixelDepthOffsetMode=PDOM_Legacy, shadingModel=MSM_Unlit, usedShadingModels=Unlit}
 nodes [
  n1 = ParticleColor()
  MaterialFunctionCall(Emissive=@n1.RGB) {materialFunction=&MF_FX_EmissiveAdjustments}
  n2 = AppendVector(
   A=DynamicParameter() {defaultValue=<rgba 0.5 0 0 0>, paramNames [UV1 U Offset, Dissolve, Camera Offset, HueShift]}.UV1 U Offset
   B=Constant()
  )
  n3 = VectorParameter() {group=02-Shape, parameterName=TrailUV_Settings, channelNames {a=PanY, b=PanX, g=ScaleY, r=ScaleX}, defaultValue=<rgba 0.1 1 -0.25 0>}
  n4 = Multiply(A=TextureCoordinate() {coordinateIndex=1}, B=AppendVector(A=@n3.ScaleX, B=@n3.ScaleX))
  n5 = Time()
  n6 = AppendVector(A=@n3.PanX, B=@n3.PanX)
  n7 = TextureSample(UVs=Add(A=Add(A=@n2, B=@n4), B=Multiply(A=@n5, B=@n6))) {
   samplerType=SAMPLERTYPE_LinearColor
   texture=&/T_InstantGravel_ShapeTrails
  }
  n8 = Multiply(A=DepthFade(FadeDistance=ScalarParameter() {defaultValue=5, group=01-Opacity, parameterName=FadeDistance}), B=@n7.R)
  n9 = OneMinus(None=TextureCoordinate())
  n10 = Multiply(
   A=ShadingPathSwitch(Default=@n8, Deferred=@n8, Forward=@n7.R, Mobile=@n7.R)
   B=SmoothStep(Value=ComponentMask(None=@n9) {r=true}) {constMax=0.1}
  ) {constA=1}
  n11 = TextureCoordinate()
  n12 = ComponentMask(None=@n11) {r=true}
  n13 = SmoothStep(Max=ScalarParameter() {defaultValue=0.05, group=01-Opacity, parameterName=FlameBaseSoftness}, Value=@n12) {constMax=0.2}
  n14 = SmoothStep(Max=ScalarParameter() {defaultValue=1, group=01-Opacity, parameterName=FlameTipSoftness}, Value=OneMinus(None=@n12)) {constMax=0.1}
  n15 = ScalarParameter() {group=01-Opacity, parameterName=SideStart}
  n16 = Add(A=@n15, B=ScalarParameter() {defaultValue=0.3, group=01-Opacity, parameterName=FlameSideSoftness})
  n17 = ComponentMask(None=@n11) {g=true}
  n18 = Multiply(A=@n17, B=OneMinus(None=@n17))
  Multiply(
   A=Multiply(A=@n1.A, B=Saturate(None=@n10))
   B=Multiply(A=Multiply(A=@n13, B=@n14), B=SmoothStep(Min=@n15, Max=@n16, Value=@n18) {constMax=0.5})
  )
 ]
}
