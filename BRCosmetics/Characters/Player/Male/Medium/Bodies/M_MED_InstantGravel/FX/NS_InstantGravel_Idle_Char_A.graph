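# Niagara character effect for the InstantGravel cosmetic body, adding socket-attached ribbon trails and sprites that respond to movement speed.
# NE_ShoulderRibbons bursts 14 partitioned ribbons, offset by per-ribbon system arrays, with interpolated trailing in gameplay; its pendulum modules are disabled.
# NE_ShoulderSprites spawns with speed and reads ribbon particles to place flipbook sprites along them.
# User parameters drive variant selection, front-end opacity and foot-hidden masking of certain sockets.
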NiagaraSystem /BRCosmetics/Characters/Player/Male/Medium/Bodies/M_MED_InstantGravel/FX/NS_InstantGravel_Idle_Char_A
dependencies {
 usedDataInterfaces [
  &NiagaraDataInterfaceCurve
  &NiagaraDataInterfaceParticleRead
  &NiagaraDataInterfaceSpriteRendererInfo
  &NiagaraDataInterfaceSkeletalMesh
  &NiagaraDataInterfaceArrayInt32
  &NiagaraDataInterfaceArrayFloat3
  &NiagaraDataInterfaceArrayFloat
  &NiagaraDataInterfaceArrayColor
 ]
 usedDynamicInputs [&Boolean_OrOperation, &Multiply_Vector]
 usedModules [
  &self:SystemSpawnScript.SystemScriptSource.SystemScriptGraph.NiagaraNodeAssignment_10.SetVariables_#1
  &self:SystemSpawnScript.SystemScriptSource.SystemScriptGraph.NiagaraNodeAssignment_0.SetVariables_#2
  &self:SystemSpawnScript.SystemScriptSource.SystemScriptGraph.NiagaraNodeAssignment_5.SetVariables_#3
  &self:SystemSpawnScript.SystemScriptSource.SystemScriptGraph.NiagaraNodeAssignment_6.SetVariables_#4
  &self:SystemSpawnScript.SystemScriptSource.SystemScriptGraph.NiagaraNodeAssignment_9.SetVariables_#5
  &self:SystemSpawnScript.SystemScriptSource.SystemScriptGraph.NiagaraNodeAssignment_11.SetVariables_#6
  &self:SystemSpawnScript.SystemScriptSource.SystemScriptGraph.NiagaraNodeAssignment_8.SetVariables_#7
  &SystemState
  &self:SystemSpawnScript.SystemScriptSource.SystemScriptGraph.NiagaraNodeAssignment_4.SetVariables_#8
  &self:SystemSpawnScript.SystemScriptSource.SystemScriptGraph.NiagaraNodeAssignment_1.SetVariables_#9
  &self:SystemSpawnScript.SystemScriptSource.SystemScriptGraph.NiagaraNodeAssignment_2.SetVariables_#10
  &self:SystemSpawnScript.SystemScriptSource.SystemScriptGraph.NiagaraNodeAssignment_3.SetVariables_#11
  &self:SystemSpawnScript.SystemScriptSource.SystemScriptGraph.NiagaraNodeAssignment_7.SetVariables_#12
 ]
}
emitters {
 NE_ShoulderRibbons {
  emitterSpawnScript [
   /SetVariables(Emitter.RandomLoopDuration:NiagaraFloat=&RandomRangeFloat(Minimum=0.2, Maximum=0.4)) {isSetParameters=true, module=SetVariables_#13}
   /SetVariables(Emitter.ParticlePerRibbons=10) {isSetParameters=true}
   /SetVariables(Emitter.RibbonNumber=14) {isSetParameters=true}
  ]
  emitterUpdateScript [
   &EmitterState(Life Cycle Mode=::NewEnumerator1, Loop Behavior=::NewEnumerator1, Loop Duration Mode=::NewEnumerator1, UseLoopDelay=false)
   &SpawnBurst_Instantaneous(
    Spawn Count=&Multiply_Int(A=$Emitter.ParticlePerRibbons, B=$Emitter.RibbonNumber)
    Use Spawn Probability=false
    UseLoopCountLimit=false
   )
   &InterpolateOverTime(
    Rate Of Change:NiagaraFloat=&MakeCustomFloatFromBool(Bool=&SetBoolByFloatComparison(A=$System.EngineOwnerSpeed, B=1), True Float=0.5, False Float=150)
    Target Float Value:NiagaraFloat=$System.EngineOwnerSpeed
   )
  ]
  particleSpawnScript [
   &InitializeParticle(
    Lifetime Max=1
    Lifetime Min=0.5
    Lifetime Mode=::NewEnumerator1
    Ribbon Width=20
    Ribbon Width Mode=::NewEnumerator1
    Sprite Size Mode=::NewEnumerator3
    Uniform Sprite Size=&RandomRangeFloat(Minimum=2, Maximum=4)
    UsePositionOffset=false
    Write Material Random=true
   )
   &PartitionParticles(
    Number Of Partitions=$Emitter.RibbonNumber
    Particles Per Partition:NiagaraInt32=$Emitter.ParticlePerRibbons
    Partition Mode:ENiagara_PartitionMode=::NewEnumerator1
   )
   /SetVariables() {isSetParameters=true}
   &SkeletalMeshLocation(
    Bone / Socket Index=&SelectIntFromArray(Direct Array Index=$Particles.PartitionParticles.Partition, Int Selection Array=$System.SocketIndexArray)
    MaskSpawnGroup=false
    Mesh Sampling Type=::NewEnumerator5
    Sampled Position Offset=&SelectVectorFromArray(Direct Array Index=$Particles.PartitionParticles.Partition, Vector Selection Array=$System.BaseOffsetArray)
    Skeletal Mesh=$System.SkeletalMesh
    Socket Sampling Mode=::NewEnumerator5
   )
   &ApplyInitialForces()
   &Color(Color=&SelectLinearColorFromArray(
     Array Sampling Mode=::NewEnumerator1
     Direct Array Index=$Particles.PartitionParticles.Partition
     Color Selection Array=$System.ColorArray
     Random Seed=?
    ))
  ]
  particleUpdateScript [
   &ParticleState(Kill Particles When Lifetime Has Elapsed=false)
   &SkeletalMeshLocation(
    Bone / Socket Index=&SelectIntFromArray(Direct Array Index=$Particles.PartitionParticles.Partition, Int Selection Array=$System.SocketIndexArray)
    MaskSpawnGroup=false
    Mesh Sampling Type=::NewEnumerator5
    Position Sampling=::NewEnumerator2
    Sampled Position Offset=&SelectVectorFromArray(Direct Array Index=$Particles.PartitionParticles.Partition, Vector Selection Array=$System.BaseOffsetArray)
    Skeletal Mesh=$System.SkeletalMesh
    Socket Sampling Mode=::NewEnumerator5
   ) {module=SkeletalMeshLocation001}
   &self:SetRibbonID(
    LinkOrder:NiagaraFloat=$Particles.PartitionParticles.Normalized Index In Partition
    RibbonId:NiagaraInt32=$Particles.PartitionParticles.Partition
   ) {module=SetRibbonID}
   &PendulumSetup(
    Module.Enable Spring Driver=true
    Pendulum Length:NiagaraFloat=&Lerp_Float(
     A=0
     B=&Lerp_Float(
      A=0
      B=25
      Alpha=&FloatFromCurve(
       FloatCurve=DataInterface() {propertyValues {Curve {keys [
           {arriveTangent=1.5557, interpMode=RCIM_Cubic, leaveTangent=1.5557, tangentMode=RCTM_User, time=0, value=0}
           {interpMode=RCIM_Cubic, tangentMode=RCTM_User, time=1, value=1}
          ]}}}
       CurveIndex=$System.SpeedFactor
      )
     )
     Alpha=$Particles.PartitionParticles.Normalized Index In Partition
    )
    Pendulum Pivot:Vector3f=$Particles.SkeletalMeshLocation001.SampledPosition
    Pendulum Rest Axis:Vector3f=&TransformVector(
     Vector=&Add_Vector(
      A=&AddVectorToPosition(
       Position=&Multiply_VectorByFloat(
        Vector=&NormalizeVector(Vector To Normalize=$Engine.Owner.Velocity, Safe Normalize=true, Safe Fallback Threshold=0, Safe Fallback Vector=<0 0 0.001>)
        Float=-4
       )
       Vector=<0 0 0>
      )
      B=&TransformVector(
       Vector=&RotateVectorByQuaternion(
        Vector To Rotate=&Lerp_Vector(
         A=<0 -1 1>
         B=<0 1 -1>
         Alpha=&MakeFloatFromInt(Int=&SelectIntFromArray(Direct Array Index=$Particles.PartitionParticles.Partition, Int Selection Array=$System.SocketIndexArray))
        )
        Quaternion=$Particles.SkeletalMeshLocation001.SampledRotation
       )
       Destination Space=::World
      )
     )
     Source Space=::World
     Destination Space=::Simulation
    )
   ) {enabled=false}
   /SetVariables(Particles.GoalPosition:NiagaraPosition=&AddVectorToPosition(
     Position=$Particles.SkeletalMeshLocation001.SampledPosition
     Vector=&Lerp_Vector(
      A=<0 0 0>
      B=&Multiply_VectorByFloat(
       Vector=&TransformVector(
        Vector=&Lerp_Vector(
         A=<0 -2 0>
         B=&FN_FrontEndVector3Modifier(Gameplay=<0 -5 2.5>, Front End=<0 -15 2.5>)
         Alpha=$System.SpeedFactor
        )
        Destination Space=::Simulation
       )
       Float=&SelectFloatFromArray(Direct Array Index=$Particles.PartitionParticles.Partition, Float Selection Array=$System.LengthArray)
      )
      Alpha=$Particles.PartitionParticles.Normalized Index In Partition
     )
    )) {isSetParameters=true}
   &InterpolateOverTime(
    Initial Vector Value:NiagaraPosition=$Particles.Position
    Rate Of Change:NiagaraFloat=&Lerp_Float(A=30, B=5, Alpha=$Particles.PartitionParticles.Normalized Index In Partition)
    Target Vector Value:NiagaraPosition=$Particles.GoalPosition
    Variable Type=::NewEnumerator2
   ) {module=InterpolateOverTime001}
   /SetVariables(Particles.Position=&FN_FrontEndVector3Modifier(
     Gameplay=&Lerp_Vector(
      A=$Particles.SkeletalMeshLocation001.SampledPosition
      B=$Particles.InterpolateOverTime001.InterpolateOverTime_V3_.Moving Average
      Alpha=&MakeCustomFloatFromBool(
       Bool=&SetBoolByIntComparison(A=$Particles.PartitionParticles.Index In Partition, B=0, Comparison Type=::NewEnumerator0)
       True Float=1
       False Float=0
      )
     )
     Front End=Linked() {linkedVariable {name=Particles.GoalPosition, type=NiagaraPosition}}
    )) {isSetParameters=true}
   /SetVariables(Particles.XOffsetValue:NiagaraFloat=&Add_Float(
     A=$Particles.XOffsetValue
     B=&Multiply_Float(A=$Engine.DeltaTime, B=&Lerp_Float(A=0, B=-1.35, Alpha=$System.SpeedFactor))
    )) {isSetParameters=true}
   /SetVariables(Particles.DistanceFade:NiagaraFloat=&OneMinusFloat(Float=&ClampFloat(Float=&Divide_Float(
       A=&Subtract_Float(
        A=&VectorLength(Vector=&Subtract_Position(A=$Particles.SkeletalMeshLocation001.SampledPosition, B=$Particles.Position))
        B=&Multiply_Float(
         A=10
         B=&SelectFloatFromArray(Direct Array Index=$Particles.PartitionParticles.Partition, Float Selection Array=$System.LengthArray)
        )
       )
       B=25
      )))) {isSetParameters=true}
   /SetVariables(Particles.RibbonU0Override=&FloatFromCurve(
     FloatCurve=DataInterface() {propertyValues {Curve {keys [<at 0 0>,<at 1 1>]}}}
     CurveIndex=$Particles.PartitionParticles.Normalized Index In Partition
    )) {isSetParameters=true}
   /SetVariables(Particles.RibbonWidth=&FloatFromCurve(
     FloatCurve=DataInterface() {propertyValues {Curve {keys [<at 0 1>,<at 1 1.5>]}}}
     CurveIndex=$Particles.PartitionParticles.Normalized Index In Partition
     Scale Curve=3
    )) {isSetParameters=true}
   &DynamicMaterialParameters(
    Index 0 Param 1=&Add_Float(A=&Multiply_FloatByInt(Float=0.5, Integer=$Particles.PartitionParticles.Partition), B=$Particles.XOffsetValue)
    Index 0 Param 2=&FloatFromCurve(
     FloatCurve=DataInterface() {propertyValues {Curve {keys [<at 0 0>,<at 0.1 1>]}}}
     CurveIndex=$Particles.PartitionParticles.Normalized Index In Partition
    )
    Index 0 Param 4=$User.VariantPicker
    Index 1 Param 1=&FloatFromCurve(FloatCurve=DataInterface() {propertyValues {Curve {keys [<at 0 0>,<at 0.7 1>]}}})
    Index 1 Param 2=1
    Index 1 Param 3=&FloatFromCurve(FloatCurve=DataInterface() {propertyValues {Curve {keys [{arriveTangent=1.9378, time=0.2, value=0.4}]}}})
    Index 1 Param 4=1
    Module.Param0WriteEnabled=true
    Module.Param0WriteEnabled1=true
    Module.Param1WriteEnabled=true
    Module.Param1WriteEnabled1=true
    Module.Param2WriteEnabled=true
    Module.Param2WriteEnabled1=true
    Module.Param3WriteEnabled=true
    Module.Param3WriteEnabled1=true
    Write Parameter Index 1=true
   )
   &SolveForcesAndVelocity(Clamp Velocity=false, Limit Acceleration=false)
   &PendulumConstraint() {enabled=false}
   &CalculateAccurateVelocity() {enabled=false}
   &ScaleColor(
    Scale Alpha=&Multiply_Float(
     A=&Multiply_Float(
      A=&FN_FrontEndFloatModifier(
       Gameplay=1
       Front End=&FloatFromCurve(
        FloatCurve=DataInterface() {propertyValues {Curve {keys [
            {interpMode=RCIM_Cubic, tangentMode=RCTM_Break, time=0.333, value=0}
            {arriveTangent=4.8959, interpMode=RCIM_Cubic, tangentMode=RCTM_Break, time=1, value=1}
           ]}}}
        CurveIndex=$User.FrontendOpacityMultiplier
       )
      )
      B=&Multiply_Float(
       A=&Multiply_Float(
        A=&FloatFromCurve(
         FloatCurve=DataInterface() {propertyValues {Curve {keys [
             {arriveTangent=-3.6618, interpMode=RCIM_Cubic, leaveTangent=-3.6618, tangentMode=RCTM_User, time=0, value=1}
             {arriveTangent=-0.1837, interpMode=RCIM_Cubic, leaveTangent=-0.1837, tangentMode=RCTM_User, time=1, value=0}
            ]}}}
         CurveIndex=$Particles.PartitionParticles.Normalized Index In Partition
         Scale Curve=2
        )
        B=&MakeCustomFloatFromBool(Bool=$System.IsStopped, True Float=0, False Float=1)
       )
       B=&FloatFromCurve(FloatCurve=DataInterface() {propertyValues {Curve {keys [<at 0 0>,<at 1 1>]}}}, CurveIndex=$System.SpeedFactor)
      )
     )
     B=$Particles.DistanceFade
    )
    ScaleA=true
    ScaleRGB=true
   )
   &ScaleColor(
    Scale Alpha=&MakeCustomFloatFromBool(
     Bool=&Boolean_AndOperation(
      Bool1=$User.bisFootHidden
      Bool2=&SetBoolByIntComparison(
       A=&SelectIntFromArray(Direct Array Index=$Particles.PartitionParticles.Partition, Int Selection Array=$System.SocketIndexArray)
       B=2
       Comparison Type=::NewEnumerator2
      )
     )
     True Float=0
     False Float=1
    )
    ScaleA=true
    ScaleRGB=true
   ) {module=ScaleColor001}
  ]
  rendererClasses [&NiagaraRibbonRendererProperties, &NiagaraSpriteRendererProperties]
  renderers [
   {rendererClass=&NiagaraRibbonRendererProperties}
   {rendererClass=&NiagaraSpriteRendererProperties, rendererIndex=1}
  ]
 }
 NE_ShoulderSprites {
  emitterSpawnScript [/SetVariables(Emitter.RandomLoopDuration:NiagaraFloat=&RandomRangeFloat(Minimum=0.2, Maximum=0.4)) {isSetParameters=true, module=SetVariables_#13}]
  emitterUpdateScript [
   &EmitterState()
   &InterpolateOverTime(
    Rate Of Change:NiagaraFloat=&MakeCustomFloatFromBool(Bool=&SetBoolByFloatComparison(A=$System.EngineOwnerSpeed, B=1), True Float=0.5, False Float=150)
    Target Float Value:NiagaraFloat=$System.EngineOwnerSpeed
   )
   /SetVariables(Emitter.GlobalSpawnRate:NiagaraFloat=&MakeCustomFloatFromBool(
     Bool=$User.Debug
     True Float=40
     False Float=&Lerp_Float(
      A=0
      B=40
      Alpha=&FloatFromCurve(
       FloatCurve=DataInterface() {propertyValues {Curve {keys [
           {arriveTangent=-0.0511, interpMode=RCIM_Cubic, leaveTangent=-0.0511, tangentMode=RCTM_User, time=0.1, value=0}
           {arriveTangent=1.4959, interpMode=RCIM_Cubic, leaveTangent=1.4959, tangentMode=RCTM_User, time=1, value=1}
          ]}}}
       CurveIndex=$System.SpeedFactor
      )
     )
    )) {isSetParameters=true}
   &SpawnRate(SpawnRate=$Emitter.GlobalSpawnRate, Use Spawn Probability=false)
  ]
  particleSpawnScript [
   &PartitionParticles(Number Of Partitions=14, Partition Mode:ENiagara_PartitionMode=::NewEnumerator5)
   &InitializeParticle(
    Color=&SelectLinearColorFromArray(
     Array Sampling Mode=::NewEnumerator1
     Direct Array Index=$Particles.PartitionParticles.Partition
     Color Selection Array=$System.ColorArray
     Random Seed=?
    )
    Lifetime Max=0.7
    Lifetime Min=0.5
    Lifetime Mode=::NewEnumerator1
    Sprite Rotation Mode=::NewEnumerator2
    Sprite Size Mode=::NewEnumerator3
    Uniform Sprite Size=&RandomRangeFloat(Minimum=2, Maximum=4)
    UsePositionOffset=false
    Write Material Random=true
   )
   &AddVelocity(Velocity=&FN_FrontEndVector3Modifier(Gameplay=<0 0 0>, Front End=<0 0 12>))
   &AddVelocity(Velocity=&RandomRangeVector(Minimum=<0 -50 0>, Maximum=<0 -100 0>)) {module=AddVelocity001}
   &SkeletalMeshLocation(
    Bone / Socket Index=&SelectIntFromArray(Direct Array Index=$Particles.PartitionParticles.Partition, Int Selection Array=$System.SocketIndexArray)
    MaskSpawnGroup=false
    Mesh Sampling Type=::NewEnumerator5
    Sampled Position Offset=&SelectVectorFromArray(Direct Array Index=$Particles.PartitionParticles.Partition, Vector Selection Array=$System.BaseOffsetArray)
    Skeletal Mesh=$System.SkeletalMesh
    Socket Sampling Mode=::NewEnumerator5
   )
   &self:StickMeshToRibbon(
    IndexToMatch:NiagaraInt32=&RandomRangeInt(Minimum=1, Maximum=2)
    Particle Attribute Reader=DataInterface() {propertyValues {EmitterBinding {bindingMode=Other, emitterName=NE_ShoulderRibbons}}}
    ParticlePerRibbons=10
    Partition:NiagaraInt32=$Particles.PartitionParticles.Partition
   ) {module=StickMeshToRibbon}
   &ApplyInitialForces()
  ]
  particleUpdateScript [
   &ParticleState()
   &ScaleSpriteSize(Uniform Curve Sprite Scale=DataInterface() {propertyValues {Curve {keys [{interpMode=RCIM_Cubic, tangentMode=RCTM_User, time=0.1, value=1}, {interpMode=RCIM_Cubic, tangentMode=RCTM_User, time=1, value=0}]}}})
   &InheritSourceMovement(
    Applied Position Scale=&Multiply_VectorByFloat(Vector=&Lerp_Vector(A=<1 1 1>, B=<0.9 0.9 0.9>, Alpha=$Particles.NormalizedAge), Float=1)
    ApplyForce=false
    ApplyPosition=true
    ApplyPositionFalloff=false
   )
   &AccelerationForce(Acceleration=<0 0 1>)
   &SolveForcesAndVelocity(Clamp Velocity=false, Limit Acceleration=false)
   &CameraOffset(Camera Offset Amount=8)
   &DynamicMaterialParameters(
    Index 0 Param 1=&Add_Float(
     A=&FloatFromCurve(
      FloatCurve=DataInterface() {propertyValues {Curve {keys [<at 0 0>,<at 1 1>]}}}
      Scale Curve=&Lerp_Float(
       A=&RandomRangeFloat(Minimum=0.1, Maximum=0.15)
       B=&RandomRangeFloat(Minimum=0.2, Maximum=0.3)
       Alpha=$Particles.ZapType
      )
     )
     B=&RandomRangeFloat(Minimum=0, Maximum=1)
    )
    Index 0 Param 2=&Multiply_Float(
     A=&Lerp_Float(A=0.2, B=0.7, Alpha=$Particles.ZapLength)
     B=&Lerp_Float(A=1, B=0.75, Alpha=$System.SpeedFactor)
    )
    Index 0 Param 4=$User.VariantPicker
    Index 1 Param 1=&FloatFromCurve(FloatCurve=DataInterface() {propertyValues {Curve {keys [<at 0 0>,<at 0.7 1>]}}})
    Index 1 Param 2=1
    Index 1 Param 3=&FloatFromCurve(FloatCurve=DataInterface() {propertyValues {Curve {keys [{arriveTangent=1.9378, time=0.2, value=0.4}]}}})
    Index 1 Param 4=1
    Module.Param0WriteEnabled=true
    Module.Param0WriteEnabled1=true
    Module.Param1WriteEnabled=true
    Module.Param1WriteEnabled1=true
    Module.Param2WriteEnabled=true
    Module.Param2WriteEnabled1=true
    Module.Param3WriteEnabled=true
    Module.Param3WriteEnabled1=true
    Write Parameter Index 1=true
   )
   &SubUVAnimation(
    Evaluation Type=::NewEnumerator0
    "Sprite Renderer "=DataInterface() {propertyValues {SpriteRenderer=&self:NE_ShoulderSprites_1.NiagaraSpriteRendererProperties_0}}
    SubUV Animation Mode:ENiagara_SubUVLookupModeV2=::NewEnumerator2
    UseEndFrame=false
    UseStartFrame=false
   )
   &ScaleColor(
    Scale Alpha=&FloatFromCurve(
     FloatCurve=DataInterface() {propertyValues {Curve {keys [
         {arriveTangent=10.5866, interpMode=RCIM_Cubic, leaveTangent=10.5866, tangentMode=RCTM_User, time=0, value=0}
         {interpMode=RCIM_Cubic, tangentMode=RCTM_User, time=0.1, value=1}
         {interpMode=RCIM_Cubic, tangentMode=RCTM_User, time=1, value=0}
        ]}}}
     Scale Curve=&FN_FrontEndFloatModifier(
      Gameplay=1
      Front End=&FloatFromCurve(
       FloatCurve=DataInterface() {propertyValues {Curve {keys [
           {interpMode=RCIM_Cubic, tangentMode=RCTM_Break, time=0.333, value=0}
           {arriveTangent=4.8959, interpMode=RCIM_Cubic, tangentMode=RCTM_Break, time=1, value=1}
          ]}}}
       CurveIndex=$User.FrontendOpacityMultiplier
      )
     )
    )
    Scale RGB=&VectorFromFloat(Value=&FloatFromCurve(FloatCurve=DataInterface() {propertyValues {Curve {keys [<at 0 5>,<at 0.25 1>]}}}))
    ScaleA=true
    ScaleRGB=true
   )
   &KillParticles(Kill Particles=&Boolean_AndOperation(
     Bool1=$User.bisFootHidden
     Bool2=&SetBoolByIntComparison(
      A=&SelectIntFromArray(Direct Array Index=$Particles.PartitionParticles.Partition, Int Selection Array=$System.SocketIndexArray)
      B=2
      Comparison Type=::NewEnumerator2
     )
    ))
  ]
  rendererClasses [&NiagaraSpriteRendererProperties]
  renderers [{rendererClass=&NiagaraSpriteRendererProperties}]
 }
}
summary {
 systemName=NS_InstantGravel_Idle_Char_A
 emitters [
  {emitterName=NE_ShoulderSprites, rendererClasses [&NiagaraSpriteRendererProperties]}
  {
   emitterName=NE_ShoulderRibbons
   rendererClasses [&NiagaraRibbonRendererProperties, &NiagaraSpriteRendererProperties]
  }
 ]
 userVariables [
  User.bisFootHidden {type=NiagaraBool}
  User.bIsFrontEnd {type=NiagaraBool}
  User.bIsFrontEndPreview {type=NiagaraBool}
  User.bisInVehicle {type=NiagaraBool}
  User.Debug {type=NiagaraBool}
  User.FrontendOpacityMultiplier {type=NiagaraFloat, defaultValue=1, description=""}
  User.IsSkydiving {type=NiagaraBool}
  User.VariantPicker {type=NiagaraFloat, defaultValue=0, description=""}
 ]
}
