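# InstantGravel material function for a character cosmetic that adds a speed-reactive glow, overriding BaseColor and Emissive behind a static switch.
# Accumulated speed, or a debug override, drives reactivity, while current speed scales a transition band built with a SphereMask.
# Offset and mask textures shape a time-looped pulse and reactive masks, with a face-accessory variant.
# Pulse opacity, emissive intensity and saturation vary with time of day, and named reroutes link the stages.
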MaterialFunction /BRCosmetics/Characters/Player/Male/Medium/Bodies/M_MED_InstantGravel/Materials/MF_InstantGravel
graph {nodes [
  n1 = Multiply(A=Time(), B=ScalarParameter() {defaultValue=0.2, group=InstantGravel_DEBUG, parameterName=IG_DEBUG_ReactiveSpeed})
  ConstantBiasScale(None=Sine(None=@n1))
  Constant() {r=0.5}
  Constant()
  n2 = GetMaterialAttributes(None=FunctionInput() {inputName=MA_IN, inputType=FunctionInput_MaterialAttributes}) {attributeGetTypes [#1]}
  n3 = NamedRerouteUsage()
  n4 = NamedRerouteUsage()
  FunctionOutput(None=StaticSwitchParameter(
    True=SetMaterialAttributes!(MaterialAttributes=@n2.MaterialAttributes, BaseColor=@n3, EmissiveColor=@n4) {attributeSetTypes [#1, #2]}
    False=@n2.MaterialAttributes
   ) {group=InstantGravel, parameterName=UseInstantGravel, sortPriority=-1}) {outputName=MA_OUT}
  n5 = Time()
  n6 = ScalarParameter() {defaultValue=0.2, group=InstantGravel, parameterName=IG_Pulse_Speed}
  n7 = NamedRerouteDeclaration(None=Frac(None=Multiply(A=@n5, B=@n6))) {name=Pulse_TimeFrac, nodeColor=<rgba 0 0.3906 1 1>}
  MaterialFunctionCall(Input=@n7, Input Low=Constant(), Input High=Constant() {r=1}, Target Low=Constant(), Target High=Constant() {r=1}) {materialFunction=&RemapValueRange}
  n8 = ScalarParameter() {defaultValue=8, group=InstantGravel, parameterName=IG_Pulse_OffsetIntensity}
  n9 = Multiply(A=@n7, B=Multiply(A=@n8) {constB=1.25}) {constB=2.6}
  n10 = Constant()
  n11 = TextureSampleParameter2D() {
   group=InstantGravel
   parameterName=IG_Offset_Texture
   samplerType=SAMPLERTYPE_Alpha
   texture=/BRCosmetics/Characters/Player/Male/Medium/Bodies/M_MED_InstantGravel/Textures/T_InstantGravel_Body_Offset
  }
  n12 = StaticSwitch(True=@n10, False=@n11.R, Value=NamedRerouteUsage())
  n13 = Multiply(A=@n8, B=@n12) {constB=5}
  n14 = Distance(A=Subtract(A=Subtract(A=@n9, B=@n13), B=NamedRerouteUsage()), B=Constant() {r=0.5})
  n15 = Saturate(None=Power(Base=OneMinus(None=@n14), Exp=ScalarParameter() {defaultValue=2.5, group=InstantGravel, parameterName=IG_Pulse_Power}) {constExponent=3.5})
  n16 = NamedRerouteUsage()
  n17 = Desaturation(None=NamedRerouteUsage(), Fraction=Constant() {r=1})
  n18 = Saturate(None=Multiply(A=Power(Base=@n17)) {constB=4})
  Multiply(A=@n15, B=Multiply(A=Multiply(B=@n16) {constA=3.5}, B=NamedRerouteDeclaration(None=@n18) {name=Luminosity, nodeColor=<rgba 0 1 0.1328 1>})) {constB=20}
  n19 = TextureSampleParameter2D() {
   group=InstantGravel
   parameterName=IG_TA_Texture
   samplerType=SAMPLERTYPE_Masks
   texture=/BRCosmetics/Characters/Player/Male/Medium/Bodies/M_MED_InstantGravel/Textures/T_InstantGravel_Body_TA
  }
  n20 = ComponentMask(None=@n19.RGB) {b=true}
  NamedRerouteDeclaration(None=StaticSwitch(True=NamedRerouteDeclaration(None=@n20) {name=FaceAcc_Base, nodeColor=<rgba 0 0.2734 1 1>}, False=@n19.G, Value=NamedRerouteUsage())) {name=FXMask, nodeColor=<rgba 0 0.5078 1 1>}
  NamedRerouteDeclaration(None=@n2.BaseColor) {name=BaseColor_IN, nodeColor=<rgba 0 0.3203 1 1>}
  n21 = Saturate(None=@n15)
  n22 = Saturate(None=Power(Base=NamedRerouteUsage()) {constExponent=1.5})
  n23 = ScalarParameter() {defaultValue=1, group=InstantGravel, parameterName=IG_PulseOpacity_Day}
  n24 = ScalarParameter() {defaultValue=0.65, group=InstantGravel, parameterName=IG_PulseOpacity_Night}
  NamedRerouteDeclaration(None=Multiply(
    A=Multiply(A=@n21, B=@n22)
    B=MaterialFunctionCall(Color_Morning=@n23, Color_Day=@n23, Color_Evening=@n24, Color_Night=@n24) {materialFunction=&Color_Time_of_Day}.Result
   )) {name=Pulse, nodeColor=<rgba 0.4766 1 0 1>}
  n25 = MaterialFunctionCall(
   Hue Shift Percentage=ScalarParameter() {defaultValue=-0.1, group=InstantGravel, parameterName=IG_Pulse_HueShift}
   Texture=NamedRerouteUsage()
  ) {materialFunction=&HueShift}
  n26 = ScalarParameter() {defaultValue=-3, group=InstantGravel, parameterName=IG_Pulse_Desaturation}
  n27 = ScalarParameter() {defaultValue=0.85, group=InstantGravel, parameterName=IG_Pulse_WhiteBlend}
  n28 = NamedRerouteUsage()
  n29 = LinearInterpolate(
   A=NamedRerouteUsage()
   B=LinearInterpolate(
    A=Desaturation(None=@n25.Result, Fraction=@n26)
    B=Constant3Vector() {constant=<rgba 1 1 1 0>}
    Alpha=LinearInterpolate(B=@n27, Alpha=@n28) {constA=0.1, constB=0.2}
   )
   Alpha=NamedRerouteUsage()
  )
  NamedRerouteDeclaration(None=@n29) {name=Debug, nodeColor=<rgba 1 0 0.7266 1>}
  n30 = VectorParameter() {group=InstantGravel, parameterName=IG_FaceAcc_ReactiveColor2, defaultValue=<rgba 1 0.6584 0 1>}
  n31 = Desaturation(None=@n30.RGB, Fraction=Constant() {r=-0.7})
  n32 = NamedRerouteUsage()
  n33 = NamedRerouteUsage()
  NamedRerouteDeclaration(None=LinearInterpolate(
    A=VectorParameter() {group=InstantGravel, parameterName=IG_FaceAcc_ReactiveColor1, defaultValue=<rgba 0.098 0 0.2083 1>}.RGB
    B=LinearInterpolate(A=@n30.RGB, B=@n31, Alpha=@n32)
    Alpha=SmoothStep(Value=@n33)
   )) {name=FaceAcc_ReactiveColor, nodeColor=<rgba 1 0 0.9609 1>}
  NamedRerouteDeclaration(None=StaticSwitch(True=NamedRerouteUsage(), False=NamedRerouteUsage(), Value=NamedRerouteUsage())) {name=ReactiveColor, nodeColor=<rgba 1 0 0.9609 1>}
  NamedRerouteDeclaration(None=StaticBoolParameter() {group=InstantGravel, parameterName=UseFaceAcc}) {name=UseFA, nodeColor=<rgba 1 0.2344 0 1>}
  n34 = Desaturation(None=@n29, Fraction=ScalarParameter() {defaultValue=-1, group=InstantGravel, parameterName=IG_EmissiveDesaturation_Day, sortPriority=10})
  n35 = Desaturation(None=@n29, Fraction=ScalarParameter() {defaultValue=-0.4, group=InstantGravel, parameterName=IG_EmissiveDesaturation_Night, sortPriority=10})
  n36 = ScalarParameter() {defaultValue=3, group=InstantGravel, parameterName=IG_EmissiveIntensity_Day, sortPriority=10}
  n37 = ScalarParameter() {defaultValue=1.5, group=InstantGravel, parameterName=IG_EmissiveIntensity_Night, sortPriority=10}
  n38 = MaterialFunctionCall(Color_Morning=@n36, Color_Day=@n36, Color_Evening=@n37, Color_Night=@n37) {materialFunction=&Color_Time_of_Day}
  n39 = Multiply(
   A=Multiply(
    A=MaterialFunctionCall(Color_Morning=@n34, Color_Day=@n34, Color_Evening=@n35, Color_Night=@n35) {materialFunction=&Color_Time_of_Day}.Result
    B=NamedRerouteDeclaration(None=@n38.Result) {name=EmissiveIntensity_TOD, nodeColor=<rgba 0.5 1 0 1>}
   )
   B=NamedRerouteUsage()
  )
  n40 = NamedRerouteUsage()
  n41 = NamedRerouteUsage()
  n42 = VectorParameter() {group=InstantGravel, parameterName=IG_Transition_ColorMulti, defaultValue=<rgba 10 10 10 1>}
  n43 = NamedRerouteUsage()
  n44 = Multiply(A=ScalarParameter() {defaultValue=0.03, group=InstantGravel, parameterName=IG_Transition_Alpha}, B=NamedRerouteUsage())
  NamedRerouteDeclaration(None=LinearInterpolate(A=Multiply(A=@n39, B=@n40), B=Multiply(A=@n41, B=@n42.RGB) {constB=5}, Alpha=Multiply(A=@n43, B=@n44))) {name=Emissive_OUT, nodeColor=<rgba 0 1 0.9062 1>}
  NamedRerouteDeclaration(None=ComponentMask(None=@n19.RGB) {g=true}) {name=FaceAcc_Pattern, nodeColor=<rgba 0.6875 0 1 1>}
  n45 = ScalarParameter() {defaultValue=8, group=InstantGravel, parameterName=IG_OffsetIntensity}
  n46 = Subtract(A=Multiply(A=NamedRerouteUsage(), B=Multiply(A=@n45) {constB=1.5}) {constB=2.6}, B=Multiply(A=@n45, B=NamedRerouteUsage()) {constB=5})
  n47 = Multiply(A=@n46, B=NamedRerouteUsage())
  NamedRerouteDeclaration(None=SmoothStep(Value=Saturate(None=@n47)) {constMax=0.9, constMin=0.15}) {name=ReactiveMask, nodeColor=<rgba 0 1 0.0391 1>}
  NamedRerouteDeclaration(None=LinearInterpolate(A=NamedRerouteUsage(), Alpha=NamedRerouteUsage()) {constB=0}) {name=BaseColor_OUT, nodeColor=<rgba 1 0 0.5859 1>}
  n48 = ScalarParameter() {defaultValue=1, group=InstantGravel, parameterName=IG_FresnelExp, sortPriority=5}
  NamedRerouteDeclaration(None=OneMinus(None=Fresnel(ExponentIn=@n48) {baseReflectFraction=0, exponent=1})) {name=Fresnel, nodeColor=<rgba 0 1 0.9531 1>}
  n49 = Constant() {r=0.15}
  n50 = ScalarParameter() {defaultValue=0.3, group=InstantGravel, parameterName=IG_OffsetSpread}
  n51 = Saturate(None=Multiply(A=SphereMask(A=@n46, B=@n49, Radius=@n50) {hardnessPercent=0}, B=NamedRerouteUsage()))
  n52 = Saturate(None=Divide(
    A=ScalarParameter() {group=InstantGravel, parameterName=IG_Speed_CCPM}
    B=ScalarParameter() {defaultValue=1000, group=InstantGravel, parameterName=IG_SpeedMax}
   ))
  NamedRerouteDeclaration(None=SmoothStep(Value=Multiply(A=@n51, B=@n52) {desc=Add speed as a factor for transition}) {constMax=0.9, constMin=0.15}) {name=TransitionMask, nodeColor=<rgba 0 1 0.6016 1>}
  n53 = StaticSwitchParameter(True=Frac(None=@n1), False=ScalarParameter() {defaultValue=1, group=InstantGravel_DEBUG, parameterName=IG_DEBUG_Reactivity}) {group=InstantGravel_DEBUG, parameterName=DebugReactivityLooping}
  n54 = Saturate(None=Divide(
    A=ScalarParameter() {group=InstantGravel, parameterName=IG_SpeedAccum_CCPM}
    B=ScalarParameter() {defaultValue=75000, group=InstantGravel, parameterName=IG_MaxSpeedAccum}
   ))
  NamedRerouteDeclaration(None=Multiply(
    A=StaticSwitchParameter(True=@n53, False=@n54) {group=InstantGravel_DEBUG, parameterName=IG_DEBUG_Toggle}
    B=ScalarParameter() {defaultValue=1, group=InstantGravel, parameterName=IG_Reactivity_Toggle}
   )) {name=Reactivity, nodeColor=<rgba 1 0.2109 0 1>}
  NamedRerouteDeclaration(None=Power(Base=@n12) {constExponent=0.5}) {name=OffsetMask, nodeColor=<rgba 0.2422 1 0 1>}
  NamedRerouteDeclaration(None=StaticSwitch(True=@n10, False=@n19.B, Value=NamedRerouteUsage())) {name=Wipe, nodeColor=<rgba 1 0.75 0 1>}
  NamedRerouteUsage()
  NamedRerouteUsage()
  NamedRerouteUsage()
  SmoothStep(Value=Constant() {r=0.7}) {desc=curve ease in and out}
 ]}
