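# Physics asset for the dangling necklace on the M_MED_InstantGravel medium male player cosmetic body.
# Small capsules on dyn_necklace_mstr, dyn_necklace_01 and dyn_necklace_02 form a two-joint chain, with constraints that allow moderate swing and very little twist.
# A larger, unconstrained capsule on spine_05 adds an upper-torso body alongside the chain, apparently as a collider for the necklace.
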PhysicsAsset /BRCosmetics/Characters/Player/Male/Medium/Bodies/M_MED_InstantGravel/Meshes/M_MED_InstantGravel_Physics_Necklace
bodies {
 dyn_necklace_01 [{shapeType=Capsule, length=3, radius=3, shapeName=dyn_necklace_01_capsule, center=<2 0 0>, rotation=<rot 90 0 0>}]
 dyn_necklace_02 [{shapeType=Capsule, length=5.9135, radius=2.5046, shapeName=dyn_necklace_02_capsule, center=<2 0 0>, rotation=<rot 90 0 0>}]
 dyn_necklace_mstr [{shapeType=Capsule, length=1, radius=1, shapeName=dyn_necklace_mstr_capsule}]
 spine_05 [{shapeType=Capsule, length=16, radius=8, center=<-3.7224 1.7774 0>, rotation=<rot 0 -90 0>}]
}
constraints [
 {bone1Name=dyn_necklace_01, bone2Name=dyn_necklace_mstr, swing1LimitDegrees=15, swing2LimitDegrees=25, twistLimitDegrees=5}
 {bone1Name=dyn_necklace_02, bone2Name=dyn_necklace_01, swing1LimitDegrees=20, swing2LimitDegrees=25, twistLimitDegrees=5}
]
