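# Secondary-motion AnimBP for the InstantGravel male medium body cosmetic.
# Copies the attached parent's pose through a Control Rig, a RigidBody node using the necklace physics asset, then AnimDynamics on four dyn_stral bones.
# The update event halves RigidBody alpha while skydiving and computes a gravity override via a deprecated function, with no visible link to the node's zero gravity.
# Older necklace AnimDynamics and constraint nodes remain disconnected from the output.
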AnimBlueprint /BRCosmetics/Characters/Player/Male/Medium/Bodies/M_MED_InstantGravel/Meshes/M_MED_InstantGravel_AnimBP
target_skeleton=&/M_MED_InstantGravel_Skeleton
graphs [
 AnimGraph {nodes [
   Anim Dynamics - Bone: dyn_necklace_03(Alpha=1, ExternalForce="0,0,0") {
    angularDampingOverride=0.99
    bOverrideAngularDamping=true
    alphaBoolBlend {blendOption=Linear}
    boundBone {boneName=dyn_necklace_03}
    componentAppliedLinearAccClamp=<10000 10000 10000>
    componentLinearAccScale=<0.3 0.3 0.5>
    componentLinearVelScale=<0.3 0.3 0.5>
    physicsBodyDefinitions [{
      boundBone {boneName=dyn_necklace_03}
      boxExtents=<15 15 15>
      constraintSetup {angularLimitsMax=<0 15 30>, angularLimitsMin=<0 -15 0>, angularTarget=<0 0 0>}
      localJointOffset=<1 0 0>
     }]
   }
   n1 = "Constraint - dyn_necklace_mstr (PPP)"!(ComponentPose, Alpha=1, ConstraintWeights_0=0.8, ConstraintWeights_1:float=0.15, ConstraintWeights_2:float=0.05) -> (Pose) {
    alphaBoolBlend {blendOption=Linear}
    boneToModify {boneName=dyn_necklace_mstr}
    constraintSetup [{targetBone {boneName=spine_05}}, {targetBone {boneName=spine_04}}, {targetBone {boneName=neck_01}}]
    constraintWeights [1,1,1]
   }
   Anim Dynamics - Bone: dyn_necklace_01(ComponentPose=@n1, Alpha=1, ExternalForce="0,0,0") {
    angularDampingOverride=0.9999
    bOverrideAngularDamping=true
    simulationSpace=BoneRelative
    alphaBoolBlend {blendOption=Linear}
    boundBone {boneName=dyn_necklace_01}
    componentAppliedLinearAccClamp=<10000 10000 10000>
    componentLinearAccScale=<0.3 0.3 0.5>
    componentLinearVelScale=<0.3 0.3 0.5>
    physicsBodyDefinitions [{
      boundBone {boneName=dyn_necklace_01}
      boxExtents=<15 15 15>
      constraintSetup {angularLimitsMax=<0 0 15>, angularLimitsMin=<0 -30 -15>, angularTarget=<0 0 0>}
      localJointOffset=<1 0 0>
     }]
    relativeSpaceBone {boneName=pelvis}
   }
   Anim Dynamics - Bone: dyn_necklace_02(Alpha=1, ExternalForce="0,0,0") {
    angularDampingOverride=0.9999
    bOverrideAngularDamping=true
    alphaBoolBlend {blendOption=Linear}
    boundBone {boneName=dyn_necklace_02}
    componentAppliedLinearAccClamp=<10000 10000 10000>
    componentLinearAccScale=<0.3 0.3 0.5>
    componentLinearVelScale=<0.3 0.3 0.5>
    physicsBodyDefinitions [{
      boundBone {boneName=dyn_necklace_02}
      boxExtents=<30 30 30>
      constraintSetup {angularLimitsMax=<0 10 15>, angularLimitsMin=<0 -10 -2>, angularTarget=<0 0 0>}
      localJointOffset=<1 0 0>
     }]
   }
   n2 = Control Rig!(Source=Copy Pose From Mesh() {bUseAttachedParent=true}, Alpha, IsCrouchingCtrl:bool=false) -> (Pose) {
    alphaBoolBlend {blendOption=Linear}
    controlRigAssetReference {blueprintRigClass=&/M_MED_InstantGravel_CtrlRig.M_MED_InstantGravel_CtrlRig_C}
   }
   n3 = RigidBody!(
    ComponentPose=Local To Component(LocalPose=@n2)
    Alpha=Get RbanAlpha() -> (RbanAlpha:double)
    OverrideWorldGravity:Vector="0,0,0"
    ExternalForce
   ) -> (Pose) {
    bClampLinearTranslationLimitToRefPose=true
    bOverrideWorldGravity=true
    overridePhysicsAsset=&/M_MED_InstantGravel_Physics_Necklace
    simulationSpace=BaseBoneSpace
    alphaBoolBlend {blendOption=Linear}
    baseBoneRef {boneName=root}
    componentAppliedLinearAccClamp=<3000 3000 3000>
    componentLinearAccScale=<0.1 0.1 0.3>
    componentLinearVelScale=<0.1 0.1 0.3>
    simSpaceSettings {maxAngularAcceleration=10, maxAngularVelocity=10, velocityScaleZ=0.5, worldAlpha=0.7}
   }
   n4 = Anim Dynamics - Bone: dyn_stral_L_01(ComponentPose=@n3, Alpha=1, ExternalForce="0,0,0") {
    angularDampingOverride=0.85
    angularSpringConstant=100
    bOverrideAngularDamping=true
    simulationSpace=RootRelative
    alphaBoolBlend {blendOption=Linear}
    boundBone {boneName=dyn_stral_L_01}
    componentAppliedLinearAccClamp=<10000 10000 10000>
    componentLinearAccScale=<0.5 0.5 0.5>
    componentLinearVelScale=<0.5 0.5 0.5>
    physicsBodyDefinitions [{
      boundBone {boneName=dyn_stral_L_01}
      boxExtents=<45 45 45>
      constraintSetup {angularLimitsMax=<10 5 70>, angularLimitsMin=<-10 -15 -70>, angularTarget=<0 0 0>}
      localJointOffset=<3 0 0>
     }]
   }
   n5 = Anim Dynamics - Bone: dyn_stral_R_01(ComponentPose=@n4, Alpha=1, ExternalForce="0,0,0") {
    angularDampingOverride=0.85
    angularSpringConstant=100
    bOverrideAngularDamping=true
    bUsePlanarLimit=false
    simulationSpace=RootRelative
    alphaBoolBlend {blendOption=Linear}
    boundBone {boneName=dyn_stral_R_01}
    componentAppliedLinearAccClamp=<10000 10000 10000>
    componentLinearAccScale=<0.5 0.5 0.5>
    componentLinearVelScale=<0.5 0.5 0.5>
    physicsBodyDefinitions [{
      boundBone {boneName=dyn_stral_R_01}
      boxExtents=<45 45 45>
      constraintSetup {angularLimitsMax=<10 0 70>, angularLimitsMin=<-10 -15 -70>, angularTarget=<0 0 0>}
      localJointOffset=<3 0 0>
     }]
    planarLimits [{drivingBone {boneName=root}, planeTransform {location=<0 0 1>, rotation=<rot 0 0 0>}}]
   }
   n6 = Anim Dynamics - Bone: dyn_stral_L_02(ComponentPose=@n5, Alpha=1, ExternalForce="0,0,0") {
    angularDampingOverride=0.8
    angularSpringConstant=100
    bOverrideAngularDamping=true
    simulationSpace=RootRelative
    alphaBoolBlend {blendOption=Linear}
    boundBone {boneName=dyn_stral_L_02}
    componentAppliedLinearAccClamp=<10000 10000 10000>
    componentLinearAccScale=<0.9 0.9 0.9>
    componentLinearVelScale=<0.9 0.9 0.9>
    physicsBodyDefinitions [{
      boundBone {boneName=dyn_stral_L_02}
      boxExtents=<35 35 35>
      constraintSetup {angularLimitsMax=<15 25 70>, angularLimitsMin=<-15 -15 -70>, angularTarget=<0 0 0>}
      localJointOffset=<2 0 0>
     }]
   }
   n7 = Anim Dynamics - Bone: dyn_stral_R_02(ComponentPose=@n6, Alpha=1, ExternalForce="0,0,0") {
    angularDampingOverride=0.8
    angularSpringConstant=100
    bOverrideAngularDamping=true
    bUsePlanarLimit=false
    simulationSpace=RootRelative
    alphaBoolBlend {blendOption=Linear}
    boundBone {boneName=dyn_stral_R_02}
    componentAppliedLinearAccClamp=<10000 10000 10000>
    componentLinearAccScale=<0.9 0.9 0.9>
    componentLinearVelScale=<0.9 0.9 0.9>
    physicsBodyDefinitions [{
      boundBone {boneName=dyn_stral_R_02}
      boxExtents=<35 35 35>
      constraintSetup {angularLimitsMax=<15 25 70>, angularLimitsMin=<-15 -15 -70>, angularTarget=<0 0 0>}
      localJointOffset=<2 0 0>
     }]
    planarLimits [{drivingBone {boneName=root}, planeTransform {location=<0 0 1>, rotation=<rot 0 0 0>}}]
   }
   Output Pose(Result=Component To Local(ComponentPose=@n7)) {layerGroup=None}
  ]}
 EventGraph {nodes [
   Event BlueprintUpdateAnimation()
   > n1 = Sequence()
   > n2 = Gravity Override State Function DEPRECATED(CharacterPartAnimInstance=Self-Reference(), GameTimeInSeconds=GetGameTimeInSeconds(), GravityOverrideParams=Get Gravity Override Params())
   > Set Gravity Override(Gravity Override=@n2)
   Set RbanAlpha(execute=@n1.then_1, RbanAlpha:double=SelectFloat(A=0.5, B=1, bPickA=Get bIsSkyDiving())) -> (Output_Get:double)
  ]}
]
variables [
 Gravity Override {type=Vector}
 Gravity Override Params {type=GravityOverrideParamsStruct, default {skydive_FloatingDown {gravityOverride=<0 -500 -900>}}}
 RbanAlpha {type=number}
]
