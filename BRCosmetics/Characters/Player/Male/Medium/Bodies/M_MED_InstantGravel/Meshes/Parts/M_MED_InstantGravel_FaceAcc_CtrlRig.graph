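# Control Rig for the InstantGravel male medium character's face accessories, fitting dynamic nose and left-ear bones to the facial bones.
# Its forward solve chains four parent constraints that affect translation and scale but leave rotation free.
# The nose bone follows mostly the left nostril with slight nose influence, while three ear bones follow the left ear at partial weight.
# Hair strand and bun bones are in the hierarchy but not driven here.
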ControlRigBlueprint /BRCosmetics/Characters/Player/Male/Medium/Bodies/M_MED_InstantGravel/Meshes/Parts/M_MED_InstantGravel_FaceAcc_CtrlRig
graphs [
 RigVMFunctionLibrary {links [], nodes []}
 RigVMModel {nodes [
   RigUnit_BeginExecution = &RigUnit_BeginExecution()
   &RigUnit_ParentConstraint(
    AdvancedSettings=(InterpolationType=Average,RotationOrderForFilter=XZY)
    Child=(Type=Bone,Name="dyn_nose")
    Filter=(TranslationFilter=(bX=True,bY=True,bZ=True),RotationFilter=(bX=false,bY=false,bZ=false),ScaleFilter=(bX=True,bY=True,bZ=True))
    Parents=((Item=(Type=Bone,Name="FACIAL_L_Nostril"),Weight=0.9),(Item=(Type=Bone,Name="FACIAL_C_NoseLower"),Weight=0),(Item=(Type=Bone,Name="FACIAL_C_Nose"),Weight=0.1))
    Weight=1
    ExecutePin=@RigUnit_BeginExecution.ExecutePin
   )
   > &RigUnit_ParentConstraint(
    AdvancedSettings=(InterpolationType=Average,RotationOrderForFilter=XZY)
    Child=(Type=Bone,Name="dyn_ear_01")
    Filter=(TranslationFilter=(bX=True,bY=True,bZ=True),RotationFilter=(bX=false,bY=false,bZ=false),ScaleFilter=(bX=True,bY=True,bZ=True))
    Parents=((Item=(Type=Bone,Name="FACIAL_L_Ear"),Weight=1))
    Weight=0.3
   )
   > &RigUnit_ParentConstraint(
    AdvancedSettings=(InterpolationType=Average,RotationOrderForFilter=XZY)
    Child=(Type=Bone,Name="dyn_ear_02")
    Filter=(TranslationFilter=(bX=True,bY=True,bZ=True),RotationFilter=(bX=false,bY=false,bZ=false),ScaleFilter=(bX=True,bY=True,bZ=True))
    Parents=((Item=(Type=Bone,Name="FACIAL_L_Ear"),Weight=1))
    Weight=0.3
   )
   > &RigUnit_ParentConstraint(
    AdvancedSettings=(InterpolationType=Average,RotationOrderForFilter=XZY)
    Child=(Type=Bone,Name="dyn_ear_03")
    Filter=(TranslationFilter=(bX=True,bY=True,bZ=True),RotationFilter=(bX=false,bY=false,bZ=false),ScaleFilter=(bX=True,bY=True,bZ=True))
    Parents=((Item=(Type=Bone,Name="FACIAL_L_Ear"),Weight=1))
    Weight=0.3
   )
  ]}
]
hierarchy [
 root {loc=<0 0 0>, rot=<0 0 0 1>}
 pelvis {parent=root, loc=<0 -0.7071 93.1191>, rot=<0.0127 0.707 -0.0127 -0.707>}
 spine_01 {parent=pelvis, loc=<6.7494 -0.6388 0>, rot=<0 0 0.034 -0.9994>}
 spine_02 {parent=spine_01, loc=<7.2518 -0.035 0>, rot=<0 0 0.0238 -0.9997>}
 spine_03 {parent=spine_02, loc=<6.0104 0 0>, rot=<0 0 0.0149 -0.9999>}
 spine_04 {parent=spine_03, loc=<8.1024 0 0>, rot=<0 0 -0.0132 -0.9999>}
 spine_05 {parent=spine_04, loc=<10.0146 0 0>, rot=<0 0 -0.0015 -1>}
 neck_01 {parent=spine_05, loc=<11.3138 0.1529 0>, rot=<0 0 -0.2003 -0.9797>}
 neck_02 {parent=neck_01, loc=<4.5627 0 0>, rot=<0 0 0.0504 -0.9987>}
 head {parent=neck_02, loc=<5.3914 0 0>, rot=<0 0 0.115 -0.9934>}
 dyn_strand_01_mstr {parent=head, loc=<7.0747 -8.0622 -1.972>, rot=<0.0029 -0.7071 -0.0029 -0.7071>}
 dyn_strand_01_01 {parent=dyn_strand_01_mstr, loc=<0 0 0>, rot=<-0.3421 -0.3254 0.4839 -0.7368>}
 dyn_strand_01_02 {parent=dyn_strand_01_01, loc=<3.8024 0 0>, rot=<-0.1426 0.5442 -0.2096 -0.7998>}
 dyn_strand_02_mstr {parent=head, loc=<6.3892 -7.9133 -1.2017>, rot=<0.0029 -0.7071 -0.0029 -0.7071>}
 dyn_strand_02_01 {parent=dyn_strand_02_mstr, loc=<0 0 0>, rot=<-0.4143 -0.3962 0.4657 -0.6741>}
 dyn_strand_02_02 {parent=dyn_strand_02_01, loc=<3.8446 0 0>, rot=<0 0 0 1>}
 dyn_strand_03_mstr {parent=head, loc=<6.5565 -8.5588 -0.6287>, rot=<0.0029 -0.7071 -0.0029 -0.7071>}
 dyn_strand_03_01 {parent=dyn_strand_03_mstr, loc=<0 0 0>, rot=<-0.5065 -0.5695 0.4618 -0.4538>}
 dyn_strand_03_02 {parent=dyn_strand_03_01, loc=<4.0424 0 0>, rot=<0 0 0 1>}
 dyn_strand_04_mstr {parent=head, loc=<7.0283 -8.1419 0.3631>, rot=<0.0029 -0.7071 -0.0029 -0.7071>}
 dyn_strand_04_01 {parent=dyn_strand_04_mstr, loc=<0 0 0>, rot=<-0.4714 -0.4742 0.5383 -0.5129>}
 dyn_strand_04_02 {parent=dyn_strand_04_01, loc=<5.1851 0 0>, rot=<0 0 0 1>}
 dyn_strand_05_mstr {parent=head, loc=<6.5982 -8.0288 1.2112>, rot=<0.0029 -0.7071 -0.0029 -0.7071>}
 dyn_strand_05_01 {parent=dyn_strand_05_mstr, loc=<0 0 0>, rot=<0.5602 0.5206 -0.4919 0.4162>}
 dyn_strand_05_02 {parent=dyn_strand_05_01, loc=<4.4798 0 0>, rot=<0 0 0 1>}
 dyn_strand_06_mstr {parent=head, loc=<7.3857 -7.9179 1.4916>, rot=<0.0029 -0.7071 -0.0029 -0.7071>}
 dyn_strand_06_01 {parent=dyn_strand_06_mstr, loc=<0 0 0>, rot=<-0.2492 -0.1942 0.9016 -0.2954>}
 dyn_strand_07_mstr {parent=head, loc=<9.8427 -8.7877 -1.6129>, rot=<0.0029 -0.7071 -0.0029 -0.7071>}
 dyn_strand_07_01 {parent=dyn_strand_07_mstr, loc=<0 0 0>, rot=<0.097 -0.1516 0.5303 -0.8285>}
 dyn_strand_08_mstr {parent=head, loc=<10.1193 -8.499 2.1159>, rot=<0.0029 -0.7071 -0.0029 -0.7071>}
 dyn_strand_08_01 {parent=dyn_strand_08_mstr, loc=<0 0 0>, rot=<0.0085 -0.0082 0.9232 -0.3842>}
 dyn_strand_08_02 {parent=dyn_strand_08_01, loc=<4.3756 0 0>, rot=<0 0 0 1>}
 dyn_strand_09_mstr {parent=head, loc=<10.8917 -9.0199 -0.9546>, rot=<0.0029 -0.7071 -0.0029 -0.7071>}
 dyn_strand_09_01 {parent=dyn_strand_09_mstr, loc=<0 0 0>, rot=<0.0868 -0.0046 0.6534 -0.752>}
 dyn_strand_09_02 {parent=dyn_strand_09_01, loc=<5.8226 0 0>, rot=<0 0 0 1>}
 dyn_strand_10_mstr {parent=head, loc=<10.8567 -9.1449 0.5768>, rot=<0.0029 -0.7071 -0.0029 -0.7071>}
 dyn_strand_10_01 {parent=dyn_strand_10_mstr, loc=<0 0 0>, rot=<-0.0517 0.0075 0.6927 -0.7193>}
 dyn_strand_11_mstr1 {parent=head, loc=<11.1654 -8.5745 -0.0359>, rot=<0.0029 -0.7071 -0.0029 -0.7071>}
 dyn_strand_11_01 {parent=dyn_strand_11_mstr1, loc=<0 0 0>, rot=<0.259 0.2734 0.7014 -0.6051>}
 dyn_strand_12_mstr {parent=head, loc=<11.2088 -8.9089 1.4724>, rot=<0.0029 -0.7071 -0.0029 -0.7071>}
 dyn_strand_12_01 {parent=dyn_strand_12_mstr, loc=<0 0 0>, rot=<0.0324 0.1594 0.773 -0.6131>}
 dyn_strand_12_02 {parent=dyn_strand_12_01, loc=<5.0458 0 0>, rot=<0 0 0 1>}
 dyn_strand_13_mstr {parent=head, loc=<11.1351 -11.1111 -0.6382>, rot=<0.0029 -0.7071 -0.0029 -0.7071>}
 dyn_strand_13_01 {parent=dyn_strand_13_mstr, loc=<0 0 0>, rot=<-0.2579 -0.2741 0.5894 -0.7148>}
 dyn_strand_13_02 {parent=dyn_strand_13_01, loc=<4.9316 0 0>, rot=<0 0 0 1>}
 dyn_bun {parent=head, loc=<9.1001 -8.8593 0.1043>, rot=<0.5021 -0.4979 0.4979 -0.5021>}
 FACIAL_C_FacialRoot {parent=head, loc=<1.9745 0.0657 0>, rot=<0.4897 -0.5101 -0.4897 -0.5101>}
 dyn_nose {parent=FACIAL_C_FacialRoot, loc=<1.254 0.8425 10.7402>, rot=<-0.6896 0 0 -0.7242>}
 FACIAL_L_Ear {parent=FACIAL_C_FacialRoot, loc=<6.165 -1.1451 1.4674>, rot=<-0.0017 -0.0013 -0.0262 -0.9997>}
 dyn_ear_03 {parent=FACIAL_L_Ear, loc=<0.7005 1.8112 -0.4625>, rot=<-0.6881 0.019 0.018 -0.7251>}
 dyn_ear_02 {parent=FACIAL_L_Ear, loc=<0.3573 2.0548 -0.144>, rot=<-0.6881 0.019 0.018 -0.7251>}
 dyn_ear_01 {parent=FACIAL_L_Ear, loc=<0.2101 2.1243 0.2808>, rot=<-0.6881 0.019 0.018 -0.7251>}
 FACIAL_C_Nose {parent=FACIAL_C_FacialRoot, loc=<-0.0012 0.4481 8.4229>, rot=<-0.1378 0 0 -0.9905>}
 FACIAL_C_NoseLower {parent=FACIAL_C_Nose, loc=<0.0012 2.3787 -0.3406>, rot=<0.1312 0 0 -0.9914>}
 FACIAL_L_Nostril {parent=FACIAL_C_Nose, loc=<1.1401 1.7775 1.0801>, rot=<0.0488 0.1305 -0.0019 -0.9902>}
]
