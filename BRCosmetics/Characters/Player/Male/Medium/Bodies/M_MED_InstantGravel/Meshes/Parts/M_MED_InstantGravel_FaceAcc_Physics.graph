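# Physics asset for the InstantGravel male medium character's face accessory, providing bodies and joints for thirteen dynamic strands attached to the head.
# Each strand has a tiny master capsule joined to the head capsule with free swing and twist, plus a larger, offset tip capsule joined to that master.
# The tip joints carry swing limits and a tight twist limit, with swing tuned on a few strands, most narrowly on strand 11.
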PhysicsAsset /BRCosmetics/Characters/Player/Male/Medium/Bodies/M_MED_InstantGravel/Meshes/Parts/M_MED_InstantGravel_FaceAcc_Physics
bodies {
 dyn_strand_01_01 [{shapeType=Capsule, length=2, radius=2, shapeName=dyn_strand_01_01_capsule, center=<5 0 0>, rotation=<rot 90 0 0>}]
 dyn_strand_01_mstr [{shapeType=Capsule, length=0.505, radius=0.505, shapeName=dyn_strand_01_mstr_capsule, rotation=<rot 0 -90 0>}]
 dyn_strand_02_01 [{shapeType=Capsule, length=2, radius=2, shapeName=dyn_strand_02_01_capsule, center=<5 0 0>, rotation=<rot 90 0 0>}]
 dyn_strand_02_mstr [{shapeType=Capsule, length=0.505, radius=0.505, shapeName=dyn_strand_02_mstr_capsule, rotation=<rot 0 -90 0>}]
 dyn_strand_03_01 [{shapeType=Capsule, length=2, radius=2, shapeName=dyn_strand_03_01_capsule, center=<5 0 0>, rotation=<rot 90 0 0>}]
 dyn_strand_03_mstr [{shapeType=Capsule, length=0.505, radius=0.505, shapeName=dyn_strand_03_mstr_capsule, rotation=<rot 0 -90 0>}]
 dyn_strand_04_01 [{shapeType=Capsule, length=2, radius=2, shapeName=dyn_strand_04_01_capsule, center=<5 0 0>, rotation=<rot 90 0 0>}]
 dyn_strand_04_mstr [{shapeType=Capsule, length=0.505, radius=0.505, shapeName=dyn_strand_04_mstr_capsule, rotation=<rot 0 -90 0>}]
 dyn_strand_05_01 [{shapeType=Capsule, length=2, radius=2, shapeName=dyn_strand_05_01_capsule, center=<4 0 0>, rotation=<rot 90 0 0>}]
 dyn_strand_05_mstr [{shapeType=Capsule, length=0.505, radius=0.505, shapeName=dyn_strand_05_mstr_capsule, rotation=<rot 0 -90 0>}]
 dyn_strand_06_01 [{shapeType=Capsule, length=2, radius=2, shapeName=dyn_strand_06_01_capsule, center=<5 0 0>, rotation=<rot 90 0 0>}]
 dyn_strand_06_mstr [{shapeType=Capsule, length=0.505, radius=0.505, shapeName=dyn_strand_06_mstr_capsule, rotation=<rot 0 -90 0>}]
 dyn_strand_07_01 [{shapeType=Capsule, length=2, radius=2, shapeName=dyn_strand_07_01_capsule, center=<5 0 0>, rotation=<rot 90 0 0>}]
 dyn_strand_07_mstr [{shapeType=Capsule, length=0.505, radius=0.505, shapeName=dyn_strand_07_mstr_capsule, rotation=<rot 0 -90 0>}]
 dyn_strand_08_01 [{shapeType=Capsule, length=2, radius=2, shapeName=dyn_strand_08_01_capsule, center=<5 0 0>, rotation=<rot 90 0 0>}]
 dyn_strand_08_mstr [{shapeType=Capsule, length=0.505, radius=0.505, shapeName=dyn_strand_08_mstr_capsule, rotation=<rot 0 -90 0>}]
 dyn_strand_09_01 [{shapeType=Capsule, length=2, radius=2, shapeName=dyn_strand_09_01_capsule, center=<4 0 0>, rotation=<rot 90 0 0>}]
 dyn_strand_09_mstr [{shapeType=Capsule, length=0.505, radius=0.505, shapeName=dyn_strand_09_mstr_capsule, rotation=<rot 0 -90 0>}]
 dyn_strand_10_01 [{shapeType=Capsule, length=2, radius=2, shapeName=dyn_strand_10_01_capsule, center=<2 0 0>, rotation=<rot 90 0 0>}]
 dyn_strand_10_mstr [{shapeType=Capsule, length=0.505, radius=0.505, shapeName=dyn_strand_10_mstr_capsule, rotation=<rot 0 -90 0>}]
 dyn_strand_11_01 [{shapeType=Capsule, length=2, radius=2, shapeName=dyn_strand_11_01_capsule, center=<2 0 0>, rotation=<rot 90 0 0>}]
 dyn_strand_11_mstr1 [{shapeType=Capsule, length=0.505, radius=0.505, shapeName=dyn_strand_11_mstr1_capsule, rotation=<rot 0 -90 0>}]
 dyn_strand_12_01 [{shapeType=Capsule, length=2, radius=2, shapeName=dyn_strand_12_01_capsule, center=<4 0 0>, rotation=<rot 90 0 0>}]
 dyn_strand_12_mstr [{shapeType=Capsule, length=0.505, radius=0.505, shapeName=dyn_strand_12_mstr_capsule, rotation=<rot 0 -90 0>}]
 dyn_strand_13_01 [{shapeType=Capsule, length=2, radius=2, shapeName=dyn_strand_13_01_capsule, center=<6 0 0>, rotation=<rot 90 0 0>}]
 dyn_strand_13_mstr [{shapeType=Capsule, length=0.505, radius=0.505, shapeName=dyn_strand_13_mstr_capsule, rotation=<rot 0 -90 0>}]
 head [{shapeType=Capsule, length=2, radius=2, rotation=<rot 90 0 0>}]
}
constraints [
 {
  bone1Name=dyn_strand_01_mstr
  bone2Name=head
  swing1LimitDegrees=45
  swing1Motion=Free
  swing2LimitDegrees=45
  swing2Motion=Free
  twistLimitDegrees=45
  twistMotion=Free
 }
 {bone1Name=dyn_strand_01_01, bone2Name=dyn_strand_01_mstr, swing1LimitDegrees=45, swing2LimitDegrees=45, twistLimitDegrees=5}
 {
  bone1Name=dyn_strand_02_mstr
  bone2Name=head
  swing1LimitDegrees=45
  swing1Motion=Free
  swing2LimitDegrees=45
  swing2Motion=Free
  twistLimitDegrees=45
  twistMotion=Free
 }
 {bone1Name=dyn_strand_02_01, bone2Name=dyn_strand_02_mstr, swing1LimitDegrees=45, swing2LimitDegrees=45, twistLimitDegrees=5}
 {
  bone1Name=dyn_strand_03_mstr
  bone2Name=head
  swing1LimitDegrees=45
  swing1Motion=Free
  swing2LimitDegrees=45
  swing2Motion=Free
  twistLimitDegrees=45
  twistMotion=Free
 }
 {bone1Name=dyn_strand_03_01, bone2Name=dyn_strand_03_mstr, swing1LimitDegrees=45, swing2LimitDegrees=45, twistLimitDegrees=5}
 {
  bone1Name=dyn_strand_04_mstr
  bone2Name=head
  swing1LimitDegrees=45
  swing1Motion=Free
  swing2LimitDegrees=45
  swing2Motion=Free
  twistLimitDegrees=45
  twistMotion=Free
 }
 {bone1Name=dyn_strand_04_01, bone2Name=dyn_strand_04_mstr, swing1LimitDegrees=45, swing2LimitDegrees=45, twistLimitDegrees=5}
 {
  bone1Name=dyn_strand_05_mstr
  bone2Name=head
  swing1LimitDegrees=45
  swing1Motion=Free
  swing2LimitDegrees=45
  swing2Motion=Free
  twistLimitDegrees=45
  twistMotion=Free
 }
 {bone1Name=dyn_strand_05_01, bone2Name=dyn_strand_05_mstr, swing1LimitDegrees=45, swing2LimitDegrees=45, twistLimitDegrees=5}
 {
  bone1Name=dyn_strand_06_mstr
  bone2Name=head
  swing1LimitDegrees=45
  swing1Motion=Free
  swing2LimitDegrees=45
  swing2Motion=Free
  twistLimitDegrees=45
  twistMotion=Free
 }
 {bone1Name=dyn_strand_06_01, bone2Name=dyn_strand_06_mstr, swing1LimitDegrees=45, swing2LimitDegrees=45, twistLimitDegrees=5}
 {bone1Name=dyn_strand_07_01, bone2Name=dyn_strand_07_mstr, swing1LimitDegrees=45, swing2LimitDegrees=50, twistLimitDegrees=5}
 {
  bone1Name=dyn_strand_07_mstr
  bone2Name=head
  swing1LimitDegrees=45
  swing1Motion=Free
  swing2LimitDegrees=45
  swing2Motion=Free
  twistLimitDegrees=45
  twistMotion=Free
 }
 {
  bone1Name=dyn_strand_08_mstr
  bone2Name=head
  swing1LimitDegrees=45
  swing1Motion=Free
  swing2LimitDegrees=45
  swing2Motion=Free
  twistLimitDegrees=45
  twistMotion=Free
 }
 {bone1Name=dyn_strand_08_01, bone2Name=dyn_strand_08_mstr, swing1LimitDegrees=45, swing2LimitDegrees=45, twistLimitDegrees=5}
 {
  bone1Name=dyn_strand_09_mstr
  bone2Name=head
  swing1LimitDegrees=45
  swing1Motion=Free
  swing2LimitDegrees=45
  swing2Motion=Free
  twistLimitDegrees=45
  twistMotion=Free
 }
 {bone1Name=dyn_strand_09_01, bone2Name=dyn_strand_09_mstr, swing1LimitDegrees=45, swing2LimitDegrees=65, twistLimitDegrees=5}
 {
  bone1Name=dyn_strand_10_mstr
  bone2Name=head
  swing1LimitDegrees=45
  swing1Motion=Free
  swing2LimitDegrees=45
  swing2Motion=Free
  twistLimitDegrees=45
  twistMotion=Free
 }
 {bone1Name=dyn_strand_10_01, bone2Name=dyn_strand_10_mstr, swing1LimitDegrees=45, swing2LimitDegrees=50, twistLimitDegrees=5}
 {
  bone1Name=dyn_strand_11_mstr1
  bone2Name=head
  swing1LimitDegrees=45
  swing1Motion=Free
  swing2LimitDegrees=45
  swing2Motion=Free
  twistLimitDegrees=45
  twistMotion=Free
 }
 {bone1Name=dyn_strand_11_01, bone2Name=dyn_strand_11_mstr1, swing1LimitDegrees=30, swing2LimitDegrees=25, twistLimitDegrees=5}
 {
  bone1Name=dyn_strand_12_mstr
  bone2Name=head
  swing1LimitDegrees=45
  swing1Motion=Free
  swing2LimitDegrees=45
  swing2Motion=Free
  twistLimitDegrees=45
  twistMotion=Free
 }
 {bone1Name=dyn_strand_12_01, bone2Name=dyn_strand_12_mstr, swing1LimitDegrees=45, swing2LimitDegrees=45, twistLimitDegrees=5}
 {
  bone1Name=dyn_strand_13_mstr
  bone2Name=head
  swing1LimitDegrees=45
  swing1Motion=Free
  swing2LimitDegrees=45
  swing2Motion=Free
  twistLimitDegrees=45
  twistMotion=Free
 }
 {bone1Name=dyn_strand_13_01, bone2Name=dyn_strand_13_mstr, swing1LimitDegrees=45, swing2LimitDegrees=50, twistLimitDegrees=5}
]
