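# Animation Blueprint for the InstantGravel male face accessory, adding secondary motion to its bones.
# It layers nose and left-ear branches from the player's Head mesh, refetched each update, over the attached parent's pose, then runs a Control Rig.
# That result, cached as 'PostDynamics' despite preceding physics, feeds a bool blend.
# The unwired bool defaults to thirteen chained, half-alpha AnimDynamics strand nodes, leaving the head-space RigidBody branch unselected.
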AnimBlueprint /BRCosmetics/Characters/Player/Male/Medium/Bodies/M_MED_InstantGravel/Meshes/Parts/M_MED_InstantGravel_FaceAcc_AnimBP
target_skeleton=&/M_MED_InstantGravel_FaceAcc_Skeleton
graphs [
 AnimGraph {nodes [
   Component To Local()
   "Constraint - dyn_nose (PP)"!(ComponentPose=Local To Component(), Alpha=1, ConstraintWeights_0=0.9, ConstraintWeights_1:float=0.1) -> (Pose) {
    alphaBoolBlend {blendOption=Linear}
    boneToModify {boneName=dyn_nose}
    constraintSetup [{targetBone {boneName=FACIAL_L_Nostril}}, {targetBone {boneName=FACIAL_C_Nose}}]
    constraintWeights [1,1]
   }
   Copy Pose From Mesh() {bUseAttachedParent=true}
   RigidBody() {alphaBoolBlend {blendOption=Linear}}
   n1 = Anim Dynamics - Bone: dyn_strand_01_01(ComponentPose=Local To Component(LocalPose=Use cached pose 'PostDynamics'()), Alpha=0.5, ExternalForce="0,0,0") {
    angularDampingOverride=0.995
    angularSpringConstant=5000
    bAngularSpring=true
    bOverrideAngularDamping=true
    simulationSpace=RootRelative
    alphaBoolBlend {blendOption=Linear}
    boundBone {boneName=dyn_strand_01_01}
    componentAppliedLinearAccClamp=<7000 7000 7000>
    componentLinearAccScale=<2.5 2.5 1.4>
    componentLinearVelScale=<2.5 2.5 1.4>
    physicsBodyDefinitions [{
      boundBone {boneName=dyn_strand_01_01}
      boxExtents=<60 30 30>
      constraintSetup {angularLimitsMax=<5 45 50>, angularLimitsMin=<-5 -45 -30>, angularTarget=<0 0 0>}
      localJointOffset=<3 0 -3>
     }]
    relativeSpaceBone {boneName=root}
   }
   n2 = Anim Dynamics - Bone: dyn_strand_02_01(ComponentPose=@n1, Alpha=0.5, ExternalForce="0,0,0") {
    angularDampingOverride=0.995
    angularSpringConstant=5000
    bAngularSpring=true
    bOverrideAngularDamping=true
    simulationSpace=RootRelative
    alphaBoolBlend {blendOption=Linear}
    boundBone {boneName=dyn_strand_02_01}
    componentAppliedLinearAccClamp=<7000 7000 7000>
    componentLinearAccScale=<2.5 2.5 1.4>
    componentLinearVelScale=<2.5 2.5 1.4>
    physicsBodyDefinitions [{
      boundBone {boneName=dyn_strand_02_01}
      boxExtents=<60 30 30>
      constraintSetup {angularLimitsMax=<5 45 35>, angularLimitsMin=<-5 -45 -35>, angularTarget=<0 0 0>}
      localJointOffset=<3 0 -3>
     }]
    relativeSpaceBone {boneName=root}
   }
   n3 = Anim Dynamics - Bone: dyn_strand_03_01(ComponentPose=@n2, Alpha=0.5, ExternalForce="0,0,0") {
    angularDampingOverride=0.995
    angularSpringConstant=5000
    bAngularSpring=true
    bOverrideAngularDamping=true
    simulationSpace=RootRelative
    alphaBoolBlend {blendOption=Linear}
    boundBone {boneName=dyn_strand_03_01}
    componentAppliedLinearAccClamp=<7000 7000 7000>
    componentLinearAccScale=<2.5 2.5 1.4>
    componentLinearVelScale=<2.5 2.5 1.4>
    physicsBodyDefinitions [{
      boundBone {boneName=dyn_strand_03_01}
      boxExtents=<60 30 30>
      constraintSetup {angularLimitsMax=<5 45 35>, angularLimitsMin=<-5 -45 -35>, angularTarget=<0 0 0>}
      localJointOffset=<3 0 -3>
     }]
    relativeSpaceBone {boneName=root}
   }
   n4 = Anim Dynamics - Bone: dyn_strand_04_01(ComponentPose=@n3, Alpha=0.5, ExternalForce="0,0,0") {
    angularDampingOverride=0.995
    angularSpringConstant=5000
    bAngularSpring=true
    bOverrideAngularDamping=true
    simulationSpace=RootRelative
    alphaBoolBlend {blendOption=Linear}
    boundBone {boneName=dyn_strand_04_01}
    componentAppliedLinearAccClamp=<7000 7000 7000>
    componentLinearAccScale=<2.5 2.5 1.4>
    componentLinearVelScale=<2.5 2.5 1.4>
    physicsBodyDefinitions [{
      boundBone {boneName=dyn_strand_04_01}
      boxExtents=<60 30 30>
      constraintSetup {angularLimitsMax=<5 45 35>, angularLimitsMin=<-5 -45 -35>, angularTarget=<0 0 0>}
      localJointOffset=<3 0 -3>
     }]
    relativeSpaceBone {boneName=root}
   }
   n5 = Anim Dynamics - Bone: dyn_strand_05_01(ComponentPose=@n4, Alpha=0.5, ExternalForce="0,0,0") {
    angularDampingOverride=0.995
    angularSpringConstant=5000
    bAngularSpring=true
    bOverrideAngularDamping=true
    simulationSpace=RootRelative
    alphaBoolBlend {blendOption=Linear}
    boundBone {boneName=dyn_strand_05_01}
    componentAppliedLinearAccClamp=<7000 7000 7000>
    componentLinearAccScale=<2.5 2.5 1.4>
    componentLinearVelScale=<2.5 2.5 1.4>
    physicsBodyDefinitions [{
      boundBone {boneName=dyn_strand_05_01}
      boxExtents=<60 30 30>
      constraintSetup {angularLimitsMax=<5 45 35>, angularLimitsMin=<-5 -45 -35>, angularTarget=<0 0 0>}
      localJointOffset=<3 0 -3>
     }]
    relativeSpaceBone {boneName=root}
   }
   n6 = Anim Dynamics - Bone: dyn_strand_06_01(ComponentPose=@n5, Alpha=0.5, ExternalForce="0,0,0") {
    angularDampingOverride=0.995
    angularSpringConstant=5000
    bAngularSpring=true
    bOverrideAngularDamping=true
    simulationSpace=RootRelative
    alphaBoolBlend {blendOption=Linear}
    boundBone {boneName=dyn_strand_06_01}
    componentAppliedLinearAccClamp=<7000 7000 7000>
    componentLinearAccScale=<2.5 2.5 1.4>
    componentLinearVelScale=<2.5 2.5 1.4>
    physicsBodyDefinitions [{
      boundBone {boneName=dyn_strand_06_01}
      boxExtents=<60 30 30>
      constraintSetup {angularLimitsMax=<5 45 45>, angularLimitsMin=<-5 -45 -35>, angularTarget=<0 0 0>}
      localJointOffset=<3 0 -3>
     }]
    relativeSpaceBone {boneName=root}
   }
   n7 = Anim Dynamics - Bone: dyn_strand_07_01(ComponentPose=@n6, Alpha=0.5, ExternalForce="0,0,0") {
    angularDampingOverride=0.995
    angularSpringConstant=5000
    bAngularSpring=true
    bOverrideAngularDamping=true
    simulationSpace=RootRelative
    alphaBoolBlend {blendOption=Linear}
    boundBone {boneName=dyn_strand_07_01}
    componentAppliedLinearAccClamp=<7000 7000 7000>
    componentLinearAccScale=<2.5 2.5 1.4>
    componentLinearVelScale=<2.5 2.5 1.4>
    physicsBodyDefinitions [{
      boundBone {boneName=dyn_strand_07_01}
      boxExtents=<60 30 30>
      constraintSetup {angularLimitsMax=<5 15 40>, angularLimitsMin=<-5 -60 -20>, angularTarget=<0 0 0>}
      localJointOffset=<3 0 -3>
     }]
    relativeSpaceBone {boneName=root}
   }
   n8 = Anim Dynamics - Bone: dyn_strand_08_01(ComponentPose=@n7, Alpha=0.5, ExternalForce="0,0,0") {
    angularDampingOverride=0.995
    angularSpringConstant=5000
    bAngularSpring=true
    bOverrideAngularDamping=true
    simulationSpace=RootRelative
    alphaBoolBlend {blendOption=Linear}
    boundBone {boneName=dyn_strand_08_01}
    componentAppliedLinearAccClamp=<7000 7000 7000>
    componentLinearAccScale=<2.5 2.5 1.4>
    componentLinearVelScale=<2.5 2.5 1.4>
    physicsBodyDefinitions [{
      boundBone {boneName=dyn_strand_08_01}
      boxExtents=<60 30 30>
      constraintSetup {angularLimitsMax=<5 15 30>, angularLimitsMin=<-5 -45 -25>, angularTarget=<0 0 0>}
      localJointOffset=<3 0 -3>
     }]
    relativeSpaceBone {boneName=root}
   }
   n9 = Anim Dynamics - Bone: dyn_strand_09_01(ComponentPose=@n8, Alpha=0.5, ExternalForce="0,0,0") {
    angularDampingOverride=0.995
    angularSpringConstant=5000
    bAngularSpring=true
    bOverrideAngularDamping=true
    simulationSpace=RootRelative
    alphaBoolBlend {blendOption=Linear}
    boundBone {boneName=dyn_strand_09_01}
    componentAppliedLinearAccClamp=<7000 7000 7000>
    componentLinearAccScale=<2.5 2.5 1.4>
    componentLinearVelScale=<2.5 2.5 1.4>
    physicsBodyDefinitions [{
      boundBone {boneName=dyn_strand_09_01}
      boxExtents=<60 30 30>
      constraintSetup {angularLimitsMax=<5 10 20>, angularLimitsMin=<-5 -60 -20>, angularTarget=<0 0 0>}
      localJointOffset=<3 0 -3>
     }]
    relativeSpaceBone {boneName=root}
   }
   n10 = Anim Dynamics - Bone: dyn_strand_10_01(ComponentPose=@n9, Alpha=0.5, ExternalForce="0,0,0") {
    angularDampingOverride=0.995
    angularSpringConstant=5000
    bAngularSpring=true
    bOverrideAngularDamping=true
    simulationSpace=RootRelative
    alphaBoolBlend {blendOption=Linear}
    boundBone {boneName=dyn_strand_10_01}
    componentAppliedLinearAccClamp=<7000 7000 7000>
    componentLinearAccScale=<2.5 2.5 1.4>
    componentLinearVelScale=<2.5 2.5 1.4>
    physicsBodyDefinitions [{
      boundBone {boneName=dyn_strand_10_01}
      boxExtents=<60 30 30>
      constraintSetup {angularLimitsMax=<5 10 30>, angularLimitsMin=<-5 -60 -30>, angularTarget=<0 0 0>}
      localJointOffset=<3 0 -3>
     }]
    relativeSpaceBone {boneName=root}
   }
   n11 = Anim Dynamics - Bone: dyn_strand_11_01(ComponentPose=@n10, Alpha=0.5, ExternalForce="0,0,0") {
    angularDampingOverride=0.995
    angularSpringConstant=5000
    bAngularSpring=true
    bOverrideAngularDamping=true
    simulationSpace=RootRelative
    alphaBoolBlend {blendOption=Linear}
    boundBone {boneName=dyn_strand_11_01}
    componentAppliedLinearAccClamp=<7000 7000 7000>
    componentLinearAccScale=<2.5 2.5 1.4>
    componentLinearVelScale=<2.5 2.5 1.4>
    physicsBodyDefinitions [{
      boundBone {boneName=dyn_strand_11_01}
      boxExtents=<60 30 30>
      constraintSetup {angularLimitsMax=<5 5 30>, angularLimitsMin=<-5 -5 -30>, angularTarget=<0 0 0>}
      localJointOffset=<3 0 -3>
     }]
    relativeSpaceBone {boneName=root}
   }
   n12 = Anim Dynamics - Bone: dyn_strand_12_01(ComponentPose=@n11, Alpha=0.5, ExternalForce="0,0,0") {
    angularDampingOverride=0.995
    angularSpringConstant=5000
    bAngularSpring=true
    bOverrideAngularDamping=true
    simulationSpace=RootRelative
    alphaBoolBlend {blendOption=Linear}
    boundBone {boneName=dyn_strand_12_01}
    componentAppliedLinearAccClamp=<7000 7000 7000>
    componentLinearAccScale=<2.5 2.5 1.4>
    componentLinearVelScale=<2.5 2.5 1.4>
    physicsBodyDefinitions [{
      boundBone {boneName=dyn_strand_12_01}
      boxExtents=<60 30 30>
      constraintSetup {angularLimitsMax=<5 15 20>, angularLimitsMin=<-5 -5 -20>, angularTarget=<0 0 0>}
      localJointOffset=<3 0 -3>
     }]
    relativeSpaceBone {boneName=root}
   }
   n13 = Anim Dynamics - Bone: dyn_strand_13_01(ComponentPose=@n12, Alpha=0.5, ExternalForce="0,0,0") {
    angularDampingOverride=0.995
    angularSpringConstant=5000
    bAngularSpring=true
    bOverrideAngularDamping=true
    simulationSpace=RootRelative
    alphaBoolBlend {blendOption=Linear}
    boundBone {boneName=dyn_strand_13_01}
    componentAppliedLinearAccClamp=<7000 7000 7000>
    componentLinearAccScale=<2.5 2.5 1.4>
    componentLinearVelScale=<2.5 2.5 1.4>
    physicsBodyDefinitions [{
      boundBone {boneName=dyn_strand_13_01}
      boxExtents=<60 30 30>
      constraintSetup {angularLimitsMax=<5 10 15>, angularLimitsMin=<-5 -10 -25>, angularTarget=<0 0 0>}
      localJointOffset=<3 0 -3>
     }]
    relativeSpaceBone {boneName=root}
   }
   n14 = RigidBody(ComponentPose=Local To Component(LocalPose=Use cached pose 'PostDynamics'())) {
    bOverrideWorldGravity=true
    overridePhysicsAsset=&/M_MED_InstantGravel_FaceAcc_Physics
    simulationSpace=BaseBoneSpace
    alphaBoolBlend {blendOption=Linear}
    baseBoneRef {boneName=head}
    componentAppliedLinearAccClamp=<3000 3000 3000>
    componentLinearAccScale=<0.21 0.21 0.21>
    componentLinearVelScale=<0.21 0.21 0.21>
    overrideWorldGravity {z=-980}
    simSpaceSettings {maxAngularAcceleration=1000, maxAngularVelocity=3, maxLinearAcceleration=250, maxLinearVelocity=500, worldAlpha=1}
   }
   Output Pose(Result=Blend Poses by bool(BlendPose_0=Component To Local(ComponentPose=@n14), BlendPose_1=Component To Local(ComponentPose=@n13)) {blendType=Linear}) {layerGroup=None}
   n15 = Get Head Mesh() -> (Head Mesh:Skeletal Mesh Component Object Reference)
   n16 = Control Rig(Source=Layered blend per bone(BasePose=Copy Pose From Mesh() {bUseAttachedParent=true}, BlendPoses_0=Copy Pose From Mesh(SourceMeshComponent=@n15)) {layerSetup [{branchFilters [{boneName=FACIAL_C_Nose}, {boneName=FACIAL_L_Ear}]}]}) {
    alphaBoolBlend {blendOption=Linear}
    controlRigAssetReference {blueprintRigClass=&/M_MED_InstantGravel_FaceAcc_CtrlRig.M_MED_InstantGravel_FaceAcc_CtrlRig_C}
   }
   Save cached pose 'PostDynamics'(Pose=@n16) {cacheName=PostDynamics}
  ]}
 EventGraph {nodes [
   Event BlueprintUpdateAnimation()
   > n1 = Cast To PlayerPawn_Athena(Object=GetOwningActor())
   > Set Head Mesh(Head Mesh:Skeletal Mesh Component Object Reference=GetSkeletalMeshForPartType(self=@n1, PartType=Head)) -> (Output_Get:Skeletal Mesh Component Object Reference)
   TryGetPawnOwner()
  ]}
]
variables [Head Mesh {type=SkeletalMeshComponent}]
